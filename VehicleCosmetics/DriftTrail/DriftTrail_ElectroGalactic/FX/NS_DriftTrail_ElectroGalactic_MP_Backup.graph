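# Niagara system for the ElectroGalactic drift trail vehicle cosmetic, named as an MP backup copy.
# Three emitters attach to skeletal mesh wheel points: lightning sprites split by spawn group, two trailing ribbons, and two scaled wheel-wrap meshes.
# FN_DisableTireFX masks each emitter per wheel, while painted-variant overrides supply colors and feed extra renderer color channels.
# User flags retune brightness and motion for painted or frontend-preview use; unpainted ribbons pulse on an engine-time sine.
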NiagaraSystem /VehicleCosmetics/DriftTrail/DriftTrail_ElectroGalactic/FX/NS_DriftTrail_ElectroGalactic_MP_Backup
dependencies {
 usedDataInterfaces [
  &NiagaraDataInterfaceArrayFloat3
  &NiagaraDataInterfaceCurve
  &NiagaraDataInterfaceArrayColor
  &NiagaraDataInterfaceSpriteRendererInfo
  &NiagaraDataInterfaceColorCurve
  &NiagaraDataInterfaceSkeletalMesh
 ]
 usedDynamicInputs [&VC_Mask_Vectors]
 usedModules [
  &self:SystemSpawnScript.SystemScriptSource.SystemScriptGraph.NiagaraNodeAssignment_0.SetVariables_#1
  &FN_UnpackVehicleFlags
  /VehicleCosmetics/FX/Niagara/Modules/VC_SetVehicleSpeedAttribs
  &SystemState
  /Game/Effects/Niagara/Modules/Vehicle/FN_UpdateVehicleAttribs
  /Game/Effects/Niagara/Modules/Vehicle/FN_VehicleFlags_Cheatcodes
 ]
}
emitters {
 NE_VC_DriftTrail_Lightning_core001 {
  emitterUpdateScript [
   &EmitterState(Inactive Response=::NewEnumerator1, Life Cycle Mode=::NewEnumerator1, Loop Duration=1, UseLoopDelay=false)
   &SpawnRate(SpawnGroup=2, SpawnRate=10, Use Spawn Probability=false) {enabled=false}
   &SpawnRate(
    SpawnGroup=1
    SpawnRate=&MakeCustomFloatFromBool(Bool=$User.bIsFrontendPreview, True Float=35, False Float=48)
    Use Spawn Probability=false
   ) {module=SpawnRate001}
   &SpawnBurst_Instantaneous(Spawn Count=10, Use Spawn Probability=false, UseLoopCountLimit=false)
   &SpawnBurst_Instantaneous(Age:NiagaraFloat=$System.LoopedAge, Spawn Count=8, Spawn Group=2, Use Spawn Probability=false, UseLoopCountLimit=false) {module=SpawnBurst_Instantaneous001}
  ]
  particleSpawnScript [
   &VC_PaintedVariantsOverride(
    Black=&MultiplyLinearColor(Linear Color=<rgba 0.001 0.001 0.001 3>, Scale Factor=<0.6 0.6 0.6 1.5>)
    BurntSienna=<rgba 1 0.325 0.1 0.7>
    Cobalt=<rgba 0 0.0333 1 1.5>
    Crimson=<rgba 1 0.04 0.056 0.85>
    DefaultColor=&SelectLinearColorFromArray(
     Array Sampling Mode=::NewEnumerator1
     Direct Array Index=&RandomRangeInt(Minimum=0, Maximum=1)
     Color Selection Array=DataInterface() {propertyValues {ColorData [<rgba 5 0 4.3333 0.5>, <rgba 0 0.9729 1 0.4>]}}
     Random Seed=?
    )
    ForestGreen=<rgba 0.4167 1 0 1>
    Grey=<rgba 0.015 0.015 0.015 1>
    Lime=<rgba 0.1667 0.25 0 1>
    Orange=<rgba 0.5 0.3333 0 0.702>
    Pink=<rgba 0.8 0.2404 0.717 1>
    Purple=<rgba 1.3333 0 2 1>
    R1 Black:LinearColor=&MultiplyLinearColor(Linear Color=<rgba 0.001 0.001 0.001 3>, Scale Factor=<0.6 0.6 0.6 1.5>)
    R1 BurntSienna:LinearColor=<rgba 0.7 0.0933 0 1.3>
    R1 Cobalt:LinearColor=<rgba 0 1.25 3 1.5>
    R1 Crimson:LinearColor=<rgba 1 0 0.0333 2>
    R1 DefaultColor=&SelectLinearColorFromArray(
     Array Sampling Mode=::NewEnumerator1
     Direct Array Index=&RandomRangeInt(Minimum=0, Maximum=1)
     Color Selection Array=DataInterface() {propertyValues {ColorData [<rgba 2.5 0 5 0.8>, <rgba 0 0.25 1 0.8>]}}
     Random Seed=?
    )
    R1 ForestGreen:LinearColor=<rgba 0.1 0.3 0 1.2>
    R1 Grey:LinearColor=<rgba 0.02 0.02 0.02 2>
    R1 Lime:LinearColor=<rgba 0.1667 0.25 0 1.5>
    R1 Orange:LinearColor=<rgba 1 0.3333 0 1.5>
    R1 Pink:LinearColor=<rgba 1 0.5 0.9258 1.25>
    R1 Purple:LinearColor=<rgba 0.825 0 1.1 1.5>
    R1 Saffron:LinearColor=<rgba 1 0.3833 0 1.5>
    R1 SkyBlue:LinearColor=<rgba 0 0.4646 0.7 1.3>
    R1 TitaniumWhite:LinearColor=<rgba 1 1 1 1.25>
    R2 Black:LinearColor=<rgba 0.01 0.01 0.01 0.5>
    R2 BurntSienna:LinearColor=<rgba 0.12 0.0371 0.0148 1>
    R2 Cobalt:LinearColor=<rgba 0.078 0.1798 1.3 1>
    R2 Crimson:LinearColor=<rgba 1.5 0.03 0.0545 1>
    R2 ForestGreen:LinearColor=<rgba 0.1625 0.5 0.05 1>
    R2 Grey:LinearColor=<rgba 0.07 0.07 0.07 1>
    R2 Lime:LinearColor=<rgba 0.2712 0.35 0.035 1>
    R2 Orange:LinearColor=<rgba 1 0.1667 0 1>
    R2 Pink:LinearColor=<rgba 0.9 0.27 0.9 1>
    R2 Purple:LinearColor=<rgba 0.1204 0.056 0.7 1>
    R2 Saffron:LinearColor=<rgba 0.5 0.345 0.035 1>
    R2 SkyBlue:LinearColor=<rgba 0.1125 0.4312 0.75 1>
    R2 TitaniumWhite:LinearColor=<rgba 0.7 0.7 0.7 1>
    Saffron=<rgba 2 1.3333 0 1>
    SkyBlue=<rgba 0 0.8333 1 0.75>
    TitaniumWhite=<rgba 0.6 0.6 0.6 0.8>
    Use R1 Painted Overrides=true
    Use R2 Painted Overrides=true
   )
   /SetVariables(Particles.RENDER TAG:NiagaraInt32=&SpawnGroup()) {isSetParameters=true}
   &InitializeParticle(
    Color=$Particles.Color
    Lifetime=/VehicleCosmetics/FX/Niagara/Modules/MergeEmitters/VC_FloatBySpawnGroup(
     SpawnGroup or ExecIndex:NiagaraInt32=0
     CustomInt:NiagaraInt32=0
     MaximumSpawnGroups:NiagaraInt32=3
     Float SG 0:NiagaraFloat=&RandomRangeFloat(Minimum=1, Maximum=2)
     Float SG 1:NiagaraFloat=&MakeCustomFloatFromBool(Bool=$User.bIsFrontendPreview, True Float=0.24, False Float=0.4)
     Float SG 2:NiagaraFloat=5.1
     Float SG 3:NiagaraFloat=0
     Float SG 4:NiagaraFloat=0
     Float SG 5:NiagaraFloat=0
    )
    Position Offset=&AddVectorToPosition(Position=<0 0 0>, Vector=$User.offset)
    Position Offset Coordinate Space=::Local
    Sprite Rotation Mode=::NewEnumerator2
    Sprite Size Mode=::NewEnumerator3
    Uniform Sprite Size=&VC_FloatBySpawnGroup(
     SpawnGroup or ExecIndex:NiagaraInt32=0
     CustomInt:NiagaraInt32=0
     MaximumSpawnGroups:NiagaraInt32=3
     Float SG 0:NiagaraFloat=/Niagara/DynamicInputs/TypeConversions/Vector2DLength(Vector:Vector2f=&Multiply_Vector2D(A=<20 20>, B=&RandomRangeVector2D(Minimum=<0.5 0.5>, Maximum=<1 1>)))
     Float SG 1:NiagaraFloat=&Vector2DLength(Vector:Vector2f=&Multiply_Vector2D(A=<24 20>, B=&RandomRangeVector2D(Minimum=<0.5 0.5>, Maximum=<1 1>)))
     Float SG 2:NiagaraFloat=44
     Float SG 3:NiagaraFloat=0
     Float SG 4:NiagaraFloat=0
     Float SG 5:NiagaraFloat=0
    )
    UsePositionOffset=true
    Write Material Random=true
   )
   &PartitionParticles()
   &SkeletalMeshLocation(
    Bone / Socket Index=&Add_Integer(A=$Particles.PartitionParticles.Partition, B=4)
    Bone / Socket Sampling Mode=::NewEnumerator5
    Interpolate Mesh Sampling=false
    MaskSpawnGroup=false
    Mesh Sampling Type=::NewEnumerator6
    Sampled Position Offset=&AddVectorToPosition(Position=<-5 0 0>, Vector=$User.offset)
    Skeletal Mesh=$System.SkeletalMesh
   )
   &AddVelocity(
    Cone Angle=&MakeCustomFloatFromBool(Bool=$User.bIsFrontendPreview, True Float=5, False Float=0)
    Cone Axis=<-1 0 0>
    Module.UseConeFalloff=false
    Module.UseVelDistribution=false
    Override Seed=false
    Velocity Mode=::NewEnumerator2
    Velocity Speed=&VC_FloatBySpawnGroup(
     SpawnGroup or ExecIndex:NiagaraInt32=0
     CustomInt:NiagaraInt32=0
     MaximumSpawnGroups:NiagaraInt32=3
     Float SG 0:NiagaraFloat=0
     Float SG 1:NiagaraFloat=&MakeCustomFloatFromBool(Bool=$User.bIsFrontendPreview, True Float=1250, False Float=500)
     Float SG 2:NiagaraFloat=0
     Float SG 3:NiagaraFloat=0
     Float SG 4:NiagaraFloat=0
     Float SG 5:NiagaraFloat=0
    )
   )
   &ShapeLocation(
    MaskSpawnGroup=false
    Override Seed=false
    Sphere Radius=&VC_FloatBySpawnGroup(
     SpawnGroup or ExecIndex:NiagaraInt32=0
     CustomInt:NiagaraInt32=0
     MaximumSpawnGroups:NiagaraInt32=3
     Float SG 0:NiagaraFloat=&RandomRangeFloat(Minimum=2, Maximum=4)
     Float SG 1:NiagaraFloat=0
     Float SG 2:NiagaraFloat=0
     Float SG 3:NiagaraFloat=0
     Float SG 4:NiagaraFloat=0
     Float SG 5:NiagaraFloat=0
    )
    UseOwnerScale=true
   )
   &SubUVAnimation(
    "Sprite Renderer "=DataInterface() {propertyValues {SpriteRenderer=&self:NE_VC_DriftTrail_Lightning_core001_0.NiagaraSpriteRendererProperties_1}}
    Start Frame Offset:NiagaraInt32=&RandomRangeInt(Minimum=0, Maximum=1)
    SubUV Animation Mode:ENiagara_SubUVLookupModeV2=::NewEnumerator1
    SubUV Curve=DataInterface() {propertyValues {Curve {keys [
        {interpMode=RCIM_Cubic, tangentMode=RCTM_User, time=0, value=0}
        {arriveTangent=0.7938, interpMode=RCIM_Cubic, leaveTangent=0.7938, tangentMode=RCTM_User, time=1, value=0.2}
       ]}}}
    UseEndFrame=false
    UseStartFrame=false
   ) {module=SubUVAnimation003}
   /VehicleCosmetics/FX/Niagara/Modules/MergeEmitters/VC_ExtraBindingColors(
    ExtraColors=2
    R1 Color:LinearColor=$Particles.R1 Color
    R2 Color:LinearColor=&MakeCustomLinearColorFromBool(
     Bool=$User.bIsPainted
     True Color=$Particles.R2 Color
     False Color=&MakeCustomLinearColorFromBool(Bool=$User.bIsFrontendPreview, True Color=<rgba 0.07 0 1.4 1>, False Color=<rgba 0.0667 0 0.8 1>)
    )
   )
  ]
  particleUpdateScript [
   &FN_DisableTireFX(
    Mask Index Attribute:NiagaraInt32=$Particles.PartitionParticles.Partition
    Mask Index FL=2
    Mask Index FR=3
    Mask Index RL=0
    Mask Index RR=1
   )
   &ParticleState()
   &SpriteRotationRate(Rotation Rate=&RandomRangeFloat(
     Minimum=&RandomRangeFloat(Minimum=-25, Maximum=-100)
     Maximum=&RandomRangeFloat(Minimum=25, Maximum=100)
    ))
   &ScaleSpriteSize(
    Scale Sprite Size Mode=::NewEnumerator0
    Uniform Scale Factor=&VC_FloatBySpawnGroup(
     SpawnGroup or ExecIndex:NiagaraInt32=0
     CustomInt:NiagaraInt32=0
     MaximumSpawnGroups:NiagaraInt32=3
     Float SG 0:NiagaraFloat=&RandomRangeFloat(
      Minimum=&FloatFromCurve(
       FloatCurve=DataInterface() {propertyValues {Curve {keys [
           {interpMode=RCIM_Cubic, tangentMode=RCTM_User, time=0, value=0.2}
           {interpMode=RCIM_Cubic, tangentMode=RCTM_User, time=0.5, value=1}
           {tangentMode=RCTM_User, time=1, value=0.2}
          ]}}}
       Scale Curve=2
      )
      Maximum=&FloatFromCurve(
       FloatCurve=DataInterface() {propertyValues {Curve {keys [{tangentMode=RCTM_User, time=0, value=0.5}, <at 0.5 1>, {tangentMode=RCTM_User, time=1, value=0.5}]}}}
       Scale Curve=2
      )
     )
     Float SG 1:NiagaraFloat=&RandomRangeFloat(
      Minimum=&FloatFromCurve(FloatCurve=DataInterface() {propertyValues {Curve {keys [<at 0 1>,<at 0.5 0.65>,<at 1 0.4>]}}}, Scale Curve=2)
      Maximum=1
     )
     Float SG 2:NiagaraFloat=1
     Float SG 3:NiagaraFloat=0
     Float SG 4:NiagaraFloat=0
     Float SG 5:NiagaraFloat=0
    )
   )
   &DynamicMaterialParameters(
    Index 0 Param 2=&FloatFromCurve(
     FloatCurve=DataInterface() {propertyValues {Curve {keys [{interpMode=RCIM_Cubic, tangentMode=RCTM_User, time=0, value=1}, {interpMode=RCIM_Cubic, tangentMode=RCTM_User, time=1, value=0}]}}}
     Scale Curve=&MakeCustomFloatFromBool(Bool=$User.bIsPainted, True Float=10, False Float=5)
    )
    Index 0 Param 3=1
    Index 0 Param 4=&RandomRangeFloat(Minimum=0.1, Maximum=0.5, Evaluation Type=::NewEnumerator1)
    Index 1 Param 1=&MakeCustomFloatFromBool(Bool=$User.bIsPainted, True Float=0.4, False Float=0.05)
    Index 2 Param 1=10
    Index 2 Param 2=&FloatFromCurve(
     FloatCurve=DataInterface() {propertyValues {Curve {keys [{interpMode=RCIM_Cubic, tangentMode=RCTM_User, time=0, value=0.8}, {interpMode=RCIM_Cubic, tangentMode=RCTM_User, time=1, value=0.2}]}}}
     Scale Curve=15
    )
    Index 2 Param 3=1
    Index 2 Param 4=&RandomRangeFloat(Minimum=0.1, Maximum=0.5, Evaluation Type=::NewEnumerator1)
    Index 3 Param 1=&MakeCustomFloatFromBool(Bool=$User.bIsPainted, True Float=0.4, False Float=0.05)
    Module.Index0Param0Write=true
    Module.Index0Param1Write=true
    Module.Index0Param2Write=true
    Module.Index0Param3Write=true
    Module.Index1Param0Write=true
    Module.Index1Param1Write=false
    Module.Index1Param2Write=false
    Module.Index1Param3Write=false
    Module.Index2Param0Write=true
    Module.Index2Param1Write=true
    Module.Index2Param2Write=true
    Module.Index2Param3Write=true
    Module.Index3Param0Write=true
    Module.Index3Param1Write=false
    Module.Index3Param2Write=false
    Module.Index3Param3Write=false
    Write Parameter Index 1=true
    Write Parameter Index 2=true
    Write Parameter Index 3=true
   )
   &ScaleColor(
    Scale Alpha=&VC_FloatBySpawnGroup(
     SpawnGroup or ExecIndex:NiagaraInt32=0
     CustomInt:NiagaraInt32=0
     MaximumSpawnGroups:NiagaraInt32=2
     Float SG 0:NiagaraFloat=&MakeCustomFloatFromBool(
      Bool=$User.bIsFrontendPreview
      True Float=1
      False Float=&MakeCustomFloatFromBool(Bool=$User.bIsPainted, True Float=2, False Float=1)
     )
     Float SG 1:NiagaraFloat=&FloatFromCurve(FloatCurve=DataInterface() {propertyValues {Curve {keys [<at 0 1>,<at 1 0>]}}})
     Float SG 2:NiagaraFloat=1
     Float SG 3:NiagaraFloat=0
     Float SG 4:NiagaraFloat=0
     Float SG 5:NiagaraFloat=0
    )
    Scale RGB=&Multiply_VectorByFloat(
     Vector=<1 1 1>
     Float=&VC_FloatBySpawnGroup(
      SpawnGroup or ExecIndex:NiagaraInt32=0
      CustomInt:NiagaraInt32=0
      MaximumSpawnGroups:NiagaraInt32=2
      Float SG 0:NiagaraFloat=&MakeCustomFloatFromBool(
       Bool=$User.bIsFrontendPreview
       True Float=1
       False Float=&MakeCustomFloatFromBool(Bool=$User.bIsPainted, True Float=3, False Float=1)
      )
      Float SG 1:NiagaraFloat=&MakeCustomFloatFromBool(
       Bool=$User.bIsFrontendPreview
       True Float=1
       False Float=&MakeCustomFloatFromBool(Bool=$User.bIsPainted, True Float=2, False Float=1)
      )
      Float SG 2:NiagaraFloat=0
      Float SG 3:NiagaraFloat=0
      Float SG 4:NiagaraFloat=0
      Float SG 5:NiagaraFloat=0
     )
    )
    ScaleA=true
    ScaleRGB=true
   )
   &SolveForcesAndVelocity(Clamp Velocity=false, Limit Acceleration=false)
   /SetVariables(Particles.R2 SCALE COLOR:LinearColor=&VC_DynamicScaleColor(
     Color to Scale:LinearColor=$Particles.R2 Color
     RGB Scale:Vector3f=&Multiply_VectorByFloat(
      Vector=<1 1 1>
      Float=&MakeCustomFloatFromBool(
       Bool=$User.bIsFrontendPreview
       True Float=&MakeCustomFloatFromBool(Bool=$User.bIsPainted, True Float=5, False Float=15)
       False Float=&MakeCustomFloatFromBool(Bool=$User.bIsPainted, True Float=5, False Float=10)
      )
     )
     Alpha Scale:NiagaraFloat=1
    )) {isSetParameters=true}
   &SkeletalMeshLocation(
    Bone / Socket Index=&Add_Integer(A=$Particles.PartitionParticles.Partition, B=4)
    Bone / Socket Sampling Mode=::NewEnumerator5
    MaskSpawnGroup=false
    Mesh Sampling Type=::NewEnumerator6
    Sampled Position Offset=&AddVectorToPosition(Position=<-5 0 0>, Vector=$User.offset)
    Skeletal Mesh=$System.SkeletalMesh
   ) {module=SkeletalMeshLocation001}
  ]
  rendererClasses [&NiagaraSpriteRendererProperties]
  renderers [
   {rendererClass=&NiagaraSpriteRendererProperties}
   {rendererClass=&NiagaraSpriteRendererProperties, rendererIndex=1}
   {rendererClass=&NiagaraSpriteRendererProperties, rendererIndex=2}
  ]
 }
 NE_VC_DriftTrail_Sec_Ribbon {
  emitterSpawnScript [/SetVariables(Emitter.ContrastBackPerColor:NiagaraFloat=&SelectFloatFromArray(
     Direct Array Index=$User.ColorVariantID
     Float Selection Array=DataInterface() {propertyValues {FloatData [1,1,1,1,4,4.5,4,4,4,8,4,6,4,4,6]}}
    )) {isSetParameters=true}]
  emitterUpdateScript [
   &EmitterState(Life Cycle Mode=::NewEnumerator1, Loop Duration=1, UseLoopDelay=false)
   &VC_RateQualityHandler(Epic Spawn Rate=90, High Spawn Rate=60, Low Spawn Rate=45, Medium Spawn Rate=45, Use Custom Rate?=true)
   &VC_SpawnRibbon_v1(Number of Ribbons=2, Spawn Time=-1, SpawnRate=$Emitter.Spawn Rate)
  ]
  particleSpawnScript [
   &VC_PaintedVariantsOverride(
    Black=<rgba 0.0015 0.0015 0.0015 0.2>
    BurntSienna=<rgba 1 0.3088 0.1233 0.4>
    Cobalt=<rgba 0.025 0.0487 0.5 0.2>
    Crimson=<rgba 1 0.0356 0.117 0.4>
    ForestGreen=<rgba 0.0822 0.35 0.035 0.435>
    Grey=<rgba 0.0356 0.0356 0.0356 0.4>
    Lime=<rgba 0.1356 0.2016 0.0097 0.302>
    Orange=<rgba 0.7 0.1228 0.021 0.18>
    Pink=<rgba 1 0.2 0.6667 0.13>
    Purple=<rgba 0.055 0.0256 0.32 0.4>
    R1 Black:LinearColor=<rgba 0.002 0.002 0.002 0.8>
    R1 BurntSienna:LinearColor=<rgba 0.025 0.0131 0.0013 0.4>
    R1 Cobalt:LinearColor=<rgba 0.0039 0.0291 0.13 0.15>
    R1 Crimson:LinearColor=<rgba 0.07 0 0.0012 0.1>
    R1 DefaultColor=<rgba 0.015 0 0.06 1>
    R1 ForestGreen:LinearColor=<rgba 0.0176 0.03 0.0003 0.3>
    R1 Grey:LinearColor=<rgba 0.009 0.009 0.009 0.4>
    R1 Lime:LinearColor=<rgba 0.0417 0.05 0 0.2>
    R1 Orange:LinearColor=<rgba 0.1 0.025 0 0.25>
    R1 Pink:LinearColor=<rgba 0.08 0.032 0.06 0.2>
    R1 Purple:LinearColor=<rgba 0.0155 0.004 0.05 0.7>
    R1 Saffron:LinearColor=<rgba 0.04 0.035 0.01 0.5>
    R1 SkyBlue:LinearColor=<rgba 0.0175 0.0437 0.07 0.25>
    R1 TitaniumWhite:LinearColor=<rgba 0.06 0.06 0.06 0.1>
    R2 BurntSienna:LinearColor=<rgba 15 4.6313 1.849 1>
    R2 Cobalt:LinearColor=<rgba 0 1.295 70 1>
    R2 Crimson:LinearColor=<rgba 30 0 0.8131 1>
    R2 DefaultColor:LinearColor=<rgba 26 0 65 10>
    R2 ForestGreen:LinearColor=<rgba 7.5 18 0 1>
    R2 Grey:LinearColor=<rgba 7 7 7 1>
    R2 Lime:LinearColor=<rgba 16.6667 25 0 1>
    R2 Orange:LinearColor=<rgba 25 8.7201 0 1>
    R2 Pink:LinearColor=<rgba 60 7.1875 40.7954 2>
    R2 Purple:LinearColor=<rgba 8.3333 5 25 1>
    R2 Saffron:LinearColor=<rgba 50 38.3333 0 1>
    R2 SkyBlue:LinearColor=<rgba 6 21.9693 30 1>
    R2 TitaniumWhite:LinearColor=<rgba 45 45 45 1>
    Saffron=<rgba 0.35 0.238 0.014 0.21>
    SkyBlue=<rgba 0.03 0.1875 0.3 0.2>
    TitaniumWhite=<rgba 0.3 0.3 0.3 0.15>
    Use R1 Painted Overrides=true
    Use R2 Painted Overrides=true
   )
   &VC_InitializeRibbon_v1(
    Custom Offset=true
    Enable Skeletal Position=true
    Lifetime=0.48
    Local Vector Offset=&AddVectorToPosition(Position=<30 0 0>, Vector=$User.offset)
    Offset Amount=1
    Ribbon Width=52
   )
   &AddVelocity(
    Velocity=<2250 0 0>
    Velocity Speed Scale=&MakeCustomFloatFromBool(
     Bool=$User.bIsFrontendPreview
     True Float=&MakeCustomFloatFromBool(Bool=$User.bIsPainted, True Float=-1, False Float=-1)
     False Float=-0.7
    )
   ) {module=AddVelocity001}
  ]
  particleUpdateScript [
   &FN_DisableTireFX(Mask Index Attribute:NiagaraInt32=$Particles.SpawnGroup, Mask Index FL=2, Mask Index FR=3, Mask Index RL=0, Mask Index RR=1)
   &VC_PaintedVariantsOverride(
    Black=&Multiply_LinearColorByFloat(Linear Color=<rgba 0.001 0 0 1>, Float=0.01)
    BurntSienna=<rgba 0.0047 0 1 0.4>
    Cobalt=<rgba 0.025 0.0487 0.5 0.2>
    Crimson=<rgba 5 0.035 0.1177 0.4>
    ForestGreen=<rgba 0.0822 0.35 0.035 0.435>
    Grey=<rgba 0.035 0.035 0.035 0.4>
    Lime=<rgba 0.1367 0.2 0.01 0.3>
    Orange=<rgba 0.7 0.1228 0.021 0.18>
    Pink=<rgba 1 0.2 0.6667 0.13>
    Purple=<rgba 0.055 0.0256 0.32 0.4>
    R3 Black:LinearColor=&ColorFromCurve(ColorCurve=DataInterface() {propertyValues {
       AlphaCurve {keys [<at 0 2 RCIM_Cubic>,<at 0.3262 0 RCIM_Cubic>]}
       BlueCurve {keys [<at 0 0.001 RCIM_Cubic>,<at 1 0 RCIM_Cubic>]}
       GreenCurve {keys [<at 0 0.001 RCIM_Cubic>,<at 1 0 RCIM_Cubic>]}
       RedCurve {keys [<at 0 0.001 RCIM_Cubic>,<at 1 0 RCIM_Cubic>]}
      }})
    R3 BurntSienna:LinearColor=&ColorFromCurve(ColorCurve=DataInterface() {propertyValues {
       AlphaCurve {keys [<at 0 2 RCIM_Cubic>,<at 0.0565 1>,<at 0.6097 0 RCIM_Cubic>]}
       BlueCurve {keys [<at 0.0312 0.0613>, {arriveTangent=0.969, interpMode=RCIM_Cubic, leaveTangent=0.969, time=0.1189, value=0.24}, <at 1 1 RCIM_Cubic>]}
       GreenCurve {keys [<at 0.0312 0.2484>, {arriveTangent=0.7758, interpMode=RCIM_Cubic, leaveTangent=0.7758, time=0.1189, value=0.9733}, <at 1 1 RCIM_Cubic>]}
       RedCurve {keys [<at 0.0312 0.5104>, {arriveTangent=0.5054, interpMode=RCIM_Cubic, leaveTangent=0.5054, time=0.1189, value=2}, <at 1 1 RCIM_Cubic>]}
      }})
    R3 Cobalt:LinearColor=&ColorFromCurve(ColorCurve=DataInterface() {propertyValues {
       AlphaCurve {keys [<at 0 2 RCIM_Cubic>,<at 0.0565 1>,<at 0.6097 0 RCIM_Cubic>]}
       BlueCurve {keys [<at 0.0076 1.8>, {arriveTangent=-0.8258, interpMode=RCIM_Cubic, time=0.102, value=6}, <at 1 1 RCIM_Cubic>]}
       GreenCurve {keys [<at 0.0076 0.1724>, {arriveTangent=0.8542, interpMode=RCIM_Cubic, time=0.102, value=0.5747}, <at 1 1 RCIM_Cubic>]}
       RedCurve {keys [<at 0.0076 0.0902>, {arriveTangent=0.9392, interpMode=RCIM_Cubic, time=0.102, value=0.3005}, <at 1 1 RCIM_Cubic>]}
      }})
    R3 Crimson:LinearColor=&ColorFromCurve(ColorCurve=DataInterface() {propertyValues {
       AlphaCurve {keys [<at 0 3 RCIM_Cubic>,<at 0.0565 1>,<at 0.6097 0 RCIM_Cubic>]}
       BlueCurve {keys [<at 0.0076 0.117>, {arriveTangent=0.8898, interpMode=RCIM_Cubic, leaveTangent=0.8898, time=0.102, value=0.5849}, <at 1 1 RCIM_Cubic>]}
       GreenCurve {keys [<at 0.0076 0.0356>, {arriveTangent=0.9718, interpMode=RCIM_Cubic, leaveTangent=0.9718, time=0.102, value=0.178}, <at 1 1 RCIM_Cubic>]}
       RedCurve {keys [<at 0.0076 1>,<at 0.102 5 RCIM_Cubic>,<at 1 1 RCIM_Cubic>]}
      }})
    R3 ForestGreen:LinearColor=&ColorFromCurve(ColorCurve=DataInterface() {propertyValues {
       AlphaCurve {keys [<at 0 3 RCIM_Cubic>,<at 0.0565 1>,<at 0.6097 0 RCIM_Cubic>]}
       BlueCurve {keys [<at 0.0076 0.1013>, {arriveTangent=0.9056, interpMode=RCIM_Cubic, leaveTangent=0.9056, time=0.102, value=1.4}, <at 1 1 RCIM_Cubic>]}
       GreenCurve {keys [<at 0.0076 1>,<at 0.102 4 RCIM_Cubic>,<at 1 1 RCIM_Cubic>]}
       RedCurve {keys [<at 0.0076 0.2341>, {arriveTangent=0.7718, interpMode=RCIM_Cubic, leaveTangent=0.7718, time=0.102, value=2.2667}, <at 1 1 RCIM_Cubic>]}
      }})
    R3 Grey:LinearColor=&ColorFromCurve(ColorCurve=DataInterface() {propertyValues {
       AlphaCurve {keys [<at 0 5 RCIM_Cubic>,<at 0.0953 1>,<at 0.6097 0 RCIM_Cubic>]}
       BlueCurve {keys [<at 0.0076 0.15>, {arriveTangent=0.8565, interpMode=RCIM_Cubic, leaveTangent=0.8565, time=0.1341, value=1.5}, <at 1 1 RCIM_Cubic>]}
       GreenCurve {keys [<at 0.0076 0.15>, {arriveTangent=0.8565, interpMode=RCIM_Cubic, leaveTangent=0.8565, time=0.1341, value=1.5}, <at 1 1 RCIM_Cubic>]}
       RedCurve {keys [<at 0.0076 0.15>, {arriveTangent=0.8565, interpMode=RCIM_Cubic, leaveTangent=0.8565, time=0.1341, value=1.5}, <at 1 1 RCIM_Cubic>]}
      }})
    R3 Lime:LinearColor=&ColorFromCurve(ColorCurve=DataInterface() {propertyValues {
       AlphaCurve {keys [<at 0 5 RCIM_Cubic>,<at 0.0953 1>,<at 0.6097 0 RCIM_Cubic>]}
       BlueCurve {keys [<at 0.0076 0.0241>, {arriveTangent=0.9833, interpMode=RCIM_Cubic, leaveTangent=0.9833, time=0.1341, value=0.1447}, <at 1 1 RCIM_Cubic>]}
       GreenCurve {keys [<at 0.0076 0.5>, {arriveTangent=0.5038, interpMode=RCIM_Cubic, leaveTangent=0.5038, time=0.1341, value=3}, <at 1 1 RCIM_Cubic>]}
       RedCurve {keys [<at 0.0076 0.3365>, {arriveTangent=0.6686, interpMode=RCIM_Cubic, leaveTangent=0.6686, time=0.1341, value=2.5241}, <at 1 1 RCIM_Cubic>]}
      }})
    R3 Orange:LinearColor=&ColorFromCurve(ColorCurve=DataInterface() {propertyValues {
       AlphaCurve {keys [<at 0 5 RCIM_Cubic>,<at 0.0953 1>,<at 0.6097 0 RCIM_Cubic>]}
       BlueCurve {keys [<at 0.0076 0.0454>, {arriveTangent=0.9619, interpMode=RCIM_Cubic, leaveTangent=0.9619, time=0.1341, value=2.4}, <at 1 1 RCIM_Cubic>]}
       GreenCurve {keys [<at 0.0076 0.409>, {arriveTangent=0.5955, interpMode=RCIM_Cubic, leaveTangent=0.5955, time=0.1341, value=4.2667}, <at 1 1 RCIM_Cubic>]}
       RedCurve {keys [<at 0.0076 1.5>, {arriveTangent=-0.5038, interpMode=RCIM_Cubic, leaveTangent=-0.5038, time=0.1341, value=8}, <at 1 1 RCIM_Cubic>]}
      }})
    R3 Pink:LinearColor=&ColorFromCurve(ColorCurve=DataInterface() {propertyValues {
       AlphaCurve {keys [<at 0 5 RCIM_Cubic>,<at 0.0953 1>,<at 0.6097 0 RCIM_Cubic>]}
       BlueCurve {keys [<at 0.0076 1.9962>, {arriveTangent=-1.0038, interpMode=RCIM_Cubic, leaveTangent=-1.0038, time=0.1341, value=2.56}, <at 1 1 RCIM_Cubic>]}
       GreenCurve {keys [<at 0.0076 0.6047>, {arriveTangent=0.3984, interpMode=RCIM_Cubic, leaveTangent=0.3984, time=0.1341, value=1.95}, <at 1 1 RCIM_Cubic>]}
       RedCurve {keys [<at 0.0076 3>, {arriveTangent=-2.0153, interpMode=RCIM_Cubic, leaveTangent=-2.0153, time=0.1341, value=3}, <at 1 1 RCIM_Cubic>]}
      }})
    R3 Purple:LinearColor=&ColorFromCurve(ColorCurve=DataInterface() {propertyValues {
       AlphaCurve {keys [<at 0 5 RCIM_Cubic>,<at 0.0953 1>,<at 0.893 0 RCIM_Cubic>]}
       BlueCurve {keys [<at 0.0076 1>,<at 0.2049 5 RCIM_Cubic>,<at 1 1 RCIM_Cubic>]}
       GreenCurve {keys [<at 0.0076 0.0791>, {arriveTangent=0.928, interpMode=RCIM_Cubic, leaveTangent=0.928, time=0.2049, value=1}, <at 1 1 RCIM_Cubic>]}
       RedCurve {keys [<at 0.0076 0.2786>, {arriveTangent=0.7269, interpMode=RCIM_Cubic, leaveTangent=0.7269, time=0.2049, value=2}, <at 1 1 RCIM_Cubic>]}
      }})
    R3 Saffron:LinearColor=&ColorFromCurve(ColorCurve=DataInterface() {propertyValues {
       AlphaCurve {keys [<at 0 5 RCIM_Cubic>,<at 0.0953 1>,<at 0.6097 0 RCIM_Cubic>]}
       BlueCurve {keys [<at 0.0076 0.15>, {arriveTangent=0.8565, interpMode=RCIM_Cubic, leaveTangent=0.8565, time=0.1341, value=0.1169}, <at 1 1 RCIM_Cubic>]}
       GreenCurve {keys [<at 0.0076 0.7875>, {arriveTangent=0.2141, interpMode=RCIM_Cubic, leaveTangent=0.2141, time=0.1341, value=2.5195}, <at 1 1 RCIM_Cubic>]}
       RedCurve {keys [<at 0.0076 1>,<at 0.1341 3 RCIM_Cubic>,<at 1 1 RCIM_Cubic>]}
      }})
    R3 SkyBlue:LinearColor=&ColorFromCurve(ColorCurve=DataInterface() {propertyValues {
       AlphaCurve {keys [<at 0 5 RCIM_Cubic>,<at 0.1206 1>,<at 0.3753 0.4793>,<at 0.6097 0 RCIM_Cubic>]}
       BlueCurve {keys [<at 0.0076 0.7>, {arriveTangent=0.3023, interpMode=RCIM_Cubic, leaveTangent=0.3023, time=0.1341, value=4}, <at 1 1 RCIM_Cubic>]}
       GreenCurve {keys [<at 0.0076 0.5422>, {arriveTangent=0.4613, interpMode=RCIM_Cubic, leaveTangent=0.4613, time=0.1341, value=2.9185}, <at 1 1 RCIM_Cubic>]}
       RedCurve {keys [<at 0.0076 0.0688>, {arriveTangent=0.9383, interpMode=RCIM_Cubic, leaveTangent=0.9383, time=0.1341, value=1.4}, <at 1 1 RCIM_Cubic>]}
      }})
    R3 TitaniumWhite:LinearColor=&ColorFromCurve(ColorCurve=DataInterface() {propertyValues {
       AlphaCurve {keys [<at 0 5 RCIM_Cubic>,<at 0.0953 1>,<at 0.6097 0 RCIM_Cubic>]}
       BlueCurve {keys [<at 0.0076 0.75>, {arriveTangent=0.2519, interpMode=RCIM_Cubic, leaveTangent=0.2519, time=0.1341, value=2.5}, <at 1 1 RCIM_Cubic>]}
       GreenCurve {keys [<at 0.0076 0.75>, {arriveTangent=0.2519, interpMode=RCIM_Cubic, leaveTangent=0.2519, time=0.1341, value=2.5}, <at 1 1 RCIM_Cubic>]}
       RedCurve {keys [<at 0.0076 0.75>, {arriveTangent=0.2519, interpMode=RCIM_Cubic, leaveTangent=0.2519, time=0.1341, value=2.5}, <at 1 1 RCIM_Cubic>]}
      }})
    Saffron=<rgba 0.35 0.238 0.014 0.21>
    SkyBlue=<rgba 0.03 0.1875 0.3 0.2>
    TitaniumWhite=<rgba 0.3 0.3 0.3 0.15>
    Use R3 Painted Overrides=true
   ) {module=VC_PaintedVariantsOverride001}
   &InheritSourceMovement(
    Applied Position Scale=&Multiply_VectorByFloat(Vector=<1 1 1>, Float=1)
    ApplyForce=false
    ApplyPosition=true
    ApplyPositionFalloff=false
   )
   &ParticleState()
   &SolveForcesAndVelocity(Clamp Velocity=false, Limit Acceleration=false)
   &ScaleRibbonWidth(Ribbon Width Scale=&FloatFromCurve(
     FloatCurve=DataInterface() {propertyValues {Curve {keys [
         {interpMode=RCIM_Cubic, tangentMode=RCTM_User, time=0, value=1.5}
         {arriveTangent=-0.7, interpMode=RCIM_Cubic, leaveTangent=-0.7, time=0.05, value=1.1}
         {interpMode=RCIM_Cubic, tangentMode=RCTM_User, time=1, value=0.8}
        ]}}}
     Scale Curve=&MakeCustomFloatFromBool(Bool=$User.bIsPainted, True Float=0.75, False Float=0.8)
    ))
   &ScaleColor(
    Scale Alpha=&FloatFromCurve(
     FloatCurve=DataInterface() {propertyValues {Curve {keys [<at 0.18 0.8>,<at 1 0>]}}}
     Scale Curve=&Lerp_Float(
      A=&FloatFromCurve(FloatCurve=DataInterface() {propertyValues {Curve {keys [<at 0 2>,<at 0.6 0>]}}})
      B=&FloatFromCurve(FloatCurve=DataInterface() {propertyValues {Curve {keys [<at 0 1.254>,<at 0.6 0>]}}})
      Alpha=HlslExpression() {hlslExpression="lerp(0, .8, (sin(Engine.Time * 3.14) + 1) * .5)"}
     )
    )
    ScaleA=true
    ScaleRGB=true
   ) {enabled=false}
   &DynamicMaterialParameters(
    Index 0 Param 2=&FloatFromCurve(
     FloatCurve=DataInterface() {propertyValues {Curve {keys [<at 0 2>,<at 1 0>]}}}
     Scale Curve=&Multiply_Float(A=$User.Scale Curve_Index 0 Param 2, B=2)
    )
    Module.Index0Param0Write=true
    Module.Index0Param1Write=true
    Module.Index0Param2Write=true
    Module.Index0Param3Write=true
   )
   &Color(
    Color=&MakeCustomLinearColorFromBool(
     Bool=$User.bIsPainted
     True Color=$Particles.R3 Color
     False Color=&Multiply_LinearColorByFloat(
      Linear Color=&ColorFromCurve(ColorCurve=DataInterface() {propertyValues {
         AlphaCurve {keys [{arriveTangent=-1.04, interpMode=RCIM_Cubic, time=0.1881, value=0.803}, {interpMode=RCIM_Cubic, leaveTangent=1.0748, time=1, value=0}]}
         BlueCurve {keys [<at 0.1829 1>,<at 0.3527 2>,<at 1 5 RCIM_Cubic>]}
         GreenCurve {keys [<at 0.1829 0.2>,<at 0.3527 1.8481>,<at 1 4.6625 RCIM_Cubic>]}
         RedCurve {keys [<at 0.1829 0.2667>,<at 0.3527 0.2>,<at 1 1 RCIM_Cubic>]}
        }})
      Float=&Lerp_Float(
       A=&FloatFromCurve(FloatCurve=DataInterface() {propertyValues {Curve {keys [<at 0 2>,<at 0.6 0>]}}})
       B=&FloatFromCurve(FloatCurve=DataInterface() {propertyValues {Curve {keys [<at 0 1.254>,<at 0.6 0>]}}})
       Alpha=HlslExpression() {hlslExpression="lerp(0, .8, (sin(Engine.Time * 3.14) + 1) * .5)"}
      )
     )
    )
    Scale Alpha=&MakeCustomFloatFromBool(Bool=$User.bIsFrontendPreview, True Float=1.5, False Float=2)
    Scale Color=&VectorFromFloat(Value=&MakeCustomFloatFromBool(Bool=$User.bIsFrontendPreview, True Float=1, False Float=0.4))
   ) {module=Color001}
   &VC_UpdateRibbon_v1(Enable Skeletal Position=true)
   /SetVariables(
    Particles.SCALECOLOR2:LinearColor=&VC_DynamicScaleColor(
     Color to Scale:LinearColor=$Particles.R1 Color
     RGB Scale:Vector3f=&VectorFromFloat(Value=&FloatFromCurve(
       FloatCurve=DataInterface() {propertyValues {Curve {keys [<at 0 1 RCIM_Cubic>, {arriveTangent=-0.0159, interpMode=RCIM_Cubic, leaveTangent=-0.0159, tangentMode=RCTM_User, time=1, value=1}]}}}
       Scale Curve=&MakeCustomFloatFromBool(Bool=$User.bIsFrontendPreview, True Float=4.5, False Float=2)
      ))
     Alpha Scale:NiagaraFloat=&FloatFromCurve(FloatCurve=DataInterface() {propertyValues {Curve {keys [
          {arriveTangent=-0.3865, interpMode=RCIM_Cubic, leaveTangent=-0.3865, tangentMode=RCTM_User, time=0.1, value=1}
          {arriveTangent=0.098, interpMode=RCIM_Cubic, leaveTangent=0.098, tangentMode=RCTM_User, time=0.8, value=0}
         ]}}})
    )
    Particles.SCALERIBBON2:NiagaraFloat=&FloatFromCurve(
     FloatCurve=DataInterface() {propertyValues {Curve {keys [
         {arriveTangent=1.6817, interpMode=RCIM_Cubic, leaveTangent=1.6817, tangentMode=RCTM_User, time=0, value=0.2}
         {arriveTangent=0.375, interpMode=RCIM_Cubic, leaveTangent=0.375, time=0.4, value=0.85}
         {arriveTangentWeight=0.1333, interpMode=RCIM_Cubic, tangentMode=RCTM_User, time=0.8, value=0.5}
        ]}}}
     Scale Curve=60
    )
   ) {isSetParameters=true}
  ]
  rendererClasses [&NiagaraRibbonRendererProperties]
  renderers [
   {rendererClass=&NiagaraRibbonRendererProperties}
   {rendererClass=&NiagaraRibbonRendererProperties, rendererIndex=1}
   {rendererClass=&NiagaraRibbonRendererProperties, rendererIndex=2}
  ]
 }
 NE_VC_DriftTrail_WheelWrap_Parent {
  emitterUpdateScript [
   &EmitterState(
    Inactive Response=::NewEnumerator1
    Life Cycle Mode=::NewEnumerator1
    Loop Behavior=::NewEnumerator1
    Loop Duration Mode=::NewEnumerator1
    UseLoopDelay=false
   )
   &SpawnBurst_Instantaneous(Spawn Count=2, Use Spawn Probability=false, UseLoopCountLimit=false)
   /SetVariables(Emitter.ISBLACK=HlslExpression() {hlslExpression="User.ColorVariantID == 2 ? true : false"}) {isSetParameters=true}
   /SetVariables(Emitter.ISNOTBLACK:NiagaraBool=&Boolean_NotOperation(Bool=$Emitter.ISBLACK)) {isSetParameters=true}
  ]
  particleSpawnScript [
   &VC_PaintedVariantsOverride(
    Black=<rgba 0.01 0.01 0.01 1>
    BurntSienna=<rgba 0.1 0.0175 0.001 1>
    Cobalt=<rgba 0.03 0.0947 1 1>
    Crimson=<rgba 0.8 0 0.0217 1>
    DefaultColor=&MakeCustomLinearColorFromBool(Bool=$User.bIsFrontendPreview, True Color=<rgba 0.05 0 1 1.5>, False Color=<rgba 0.05 0 1 1.5>)
    ForestGreen=<rgba 0.1437 0.5 0.025 0.5>
    Grey=<rgba 0.08 0.08 0.08 1>
    Lime=<rgba 0.5267 0.7 0.007 0.3>
    Orange=<rgba 0.7 0.14 0 1>
    Pink=<rgba 0.5052 0.1516 0.3766 0.9>
    Purple=<rgba 0.0906 0 0.6795 1>
    Saffron=<rgba 0.9 0.525 0 0.3>
    SkyBlue=<rgba 0.0268 0.3659 0.5365 0.7>
    TitaniumWhite=<rgba 0.6 0.6 0.6 0.5>
   )
   &InitializeParticle(
    Color Mode=::NewEnumerator0
    Mesh Scale Mode=::NewEnumerator4
    Sprite Rotation Mode=::NewEnumerator2
    Sprite Size Mode=::NewEnumerator3
    UsePositionOffset=false
    Write Material Random=true
   )
   &PartitionParticles()
   &SkeletalMeshLocation(
    Bone / Socket Index=&Add_Integer(A=$Particles.PartitionParticles.Partition, B=4)
    Bone / Socket Sampling Mode=::NewEnumerator5
    Interpolate Mesh Sampling=false
    MaskSpawnGroup=false
    Mesh Sampling Type=::NewEnumerator6
    Sampled Position Offset=&SelectVectorFromArray(
     Direct Array Index=$Particles.PartitionParticles.Partition
     Vector Selection Array=DataInterface() {propertyValues {FloatData [<0 1 0>, <0 1 0>]}}
    )
    Skeletal Mesh=$System.SkeletalMesh
   )
  ]
  particleUpdateScript [
   &FN_DisableTireFX(
    Mask Index Attribute:NiagaraInt32=$Particles.PartitionParticles.Partition
    Mask Index FL=2
    Mask Index FR=3
    Mask Index RL=0
    Mask Index RR=1
   )
   &VC_WheelWrapScale(Wheel Index Attribute:NiagaraInt32=&Add_Integer(A=$Particles.PartitionParticles.Partition, B=2))
   &DynamicMaterialParameters(
    Index 2 Param 1=&SelectFloatFromArray(Direct Array Index=$Particles.PartitionParticles.Partition, Float Selection Array=DataInterface() {propertyValues {FloatData [-3,3]}})
    Index 2 Param 2=&Multiply_Float(
     A=$User.WheelBoltIntensity
     B=&MakeCustomFloatFromBool(Bool=$User.bIsFrontendPreview, True Float=2, False Float=2)
    )
    Index 2 Param 3=&FloatFromCurve(
     FloatCurve=DataInterface() {propertyValues {Curve {keys [<at 0 1 RCIM_Cubic>,<at 0.5 1 RCIM_Cubic>,<at 1 1 RCIM_Cubic>]}}}
     CurveIndex=&Multiply_Float(A=$Emitter.Age, B=0.1)
     Scale Curve=&MakeCustomFloatFromBool(Bool=$User.bIsPainted, True Float=$User.CenterBoltLightnessPower, False Float=3)
    )
    Index 2 Param 4=&MakeCustomFloatFromBool(Bool=$User.bIsPainted, True Float=150, False Float=150)
    Module.Index2Param0Write=true
    Module.Index2Param1Write=true
    Module.Index2Param2Write=true
    Module.Index2Param3Write=true
    Write Parameter Index 0=false
    Write Parameter Index 2=true
   )
   &ScaleColor(
    Scale Alpha=&MakeCustomFloatFromBool(Bool=$User.bIsFrontendPreview, True Float=10, False Float=10)
    Scale RGB=&Multiply_VectorByFloat(
     Vector=<1 1 1>
     Float=&FloatFromCurve(
      FloatCurve=DataInterface() {propertyValues {Curve {keys [
          {interpMode=RCIM_Cubic, tangentMode=RCTM_User, time=0, value=1}
          {interpMode=RCIM_Cubic, tangentMode=RCTM_User, time=0.5, value=0.5}
          {tangentMode=RCTM_User, time=1, value=1}
         ]}}}
      Scale Curve=&MakeCustomFloatFromBool(Bool=$User.bIsFrontendPreview, True Float=&RandomRangeFloat(Minimum=4, Maximum=4), False Float=1)
     )
    )
    ScaleA=true
    ScaleRGB=true
   )
   &ParticleState(Kill Particles When Lifetime Has Elapsed=false, Loop Particles Lifetime=true)
   &UpdateMeshOrientation(Rotation Coordinate Space=::NewEnumerator2, Rotation Vector=<0 1 0>) {enabled=false}
   &SkeletalMeshLocation(
    Bone / Socket Index=&Add_Integer(A=$Particles.PartitionParticles.Partition, B=4)
    Bone / Socket Sampling Mode=::NewEnumerator5
    MaskSpawnGroup=false
    Mesh Sampling Type=::NewEnumerator6
    Sampled Position Offset=&SelectVectorFromArray(
     Direct Array Index=$Particles.PartitionParticles.Partition
     Vector Selection Array=DataInterface() {propertyValues {FloatData [<0 1 0>, <0 1 0>]}}
    )
    Skeletal Mesh=$System.SkeletalMesh
   ) {module=SkeletalMeshLocation001}
   &SolveForcesAndVelocity(Clamp Velocity=false, Limit Acceleration=false)
  ]
  rendererClasses [&NiagaraMeshRendererProperties]
  renderers [{rendererClass=&NiagaraMeshRendererProperties, rendererIndex=0}]
 }
}
summary {
 systemName=NS_DriftTrail_ElectroGalactic_MP_Backup
 emitters [
  {emitterName=NE_VC_DriftTrail_WheelWrap_Parent, rendererClasses [&NiagaraMeshRendererProperties]}
  {emitterName=NE_VC_DriftTrail_Lightning_core001, rendererClasses [&NiagaraSpriteRendererProperties]}
  {emitterName=NE_VC_DriftTrail_Sec_Ribbon, rendererClasses [&NiagaraRibbonRendererProperties]}
 ]
 userVariables [
  User.bIsFrontendPreview {type=NiagaraBool}
  User.bIsPainted {type=NiagaraBool}
  User.CenterBoltLightnessPower {type=NiagaraFloat, defaultValue=1, description=""}
  User.ColorVariantID {type=NiagaraInt32}
  User.offset {type=Vector3f, description="offset from wheels(set through cosmetic template)", defaultValue {struct=Vector3f, value=<0 0 0>}}
  User.PaintedColor {type=LinearColor, description="", defaultValue {struct=LinearColor, value=<rgba 1 1 1 1>}}
  User.PaintedScalar {type=NiagaraFloat, defaultValue=1, description=""}
  User.Scale Curve_Index 0 Param 2 {type=NiagaraFloat, defaultValue=5, description=""}
  User.VehicleFlags {type=NiagaraInt32, defaultValue=806289408}
  User.Wheel_Radius_Array {
   type=NiagaraDataInterfaceArrayFloat
   defaultValue {
    struct=NiagaraExt_VariableValue_DataInterface
    value {
     dataInterface=&self:NiagaraDataInterfaceArrayFloat_1
     dataInterfaceClass=&NiagaraDataInterfaceArrayFloat
    }
   }
  }
  User.Wheel_Width_Array {
   type=NiagaraDataInterfaceArrayFloat
   defaultValue {
    struct=NiagaraExt_VariableValue_DataInterface
    value {
     dataInterface=&self:NiagaraDataInterfaceArrayFloat_0
     dataInterfaceClass=&NiagaraDataInterfaceArrayFloat
    }
   }
  }
  User.WheelBoltIntensity {type=NiagaraFloat, defaultValue=1, description=""}
  User.WheelTouchingArray {
   type=NiagaraDataInterfaceArrayFloat
   defaultValue {
    struct=NiagaraExt_VariableValue_DataInterface
    value {
     dataInterface=&self:NiagaraDataInterfaceArrayFloat_2
     dataInterfaceClass=&NiagaraDataInterfaceArrayFloat
    }
   }
  }
 ]
}
