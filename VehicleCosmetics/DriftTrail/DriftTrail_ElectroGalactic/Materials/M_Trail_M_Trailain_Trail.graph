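# Translucent unlit Niagara ribbon material for the ElectroGalactic drift trail, drawing particle-colored smoke with glowing lightning bolts.
# Panned cloud noise and smoke textures form the body; flow-warped UVs scroll three lightning channels, thresholded and boosted by dynamic LBright and Bolt Intensity.
# Bolts add to smoke through an emissive adjustment function, while smoke, particle alpha and UV edge masks drive depth-faded opacity.
# The StarsPaint desaturation blend feeds nothing.
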Material /VehicleCosmetics/DriftTrail/DriftTrail_ElectroGalactic/Materials/M_Trail_M_Trailain_Trail
graph {
 properties {
  bAutomaticallySetUsageInEditor=false
  bUsedWithNiagaraRibbons=true
  blendMode=BLEND_Translucent
  shadingModel=MSM_Unlit
  usedShadingModels=Unlit
 }
 nodes [
  Constant3Vector() {constant=<rgba 1 1 1 0>}
  MaterialExpressionVertexColor_1 = ParticleColor()
  n1 = Panner(Coordinate=TextureCoordinate()) {speedX=-1, speedY=0.1}
  n2 = TextureCoordinate()
  n3 = ComponentMask(None=@n2) {g=true}
  n4 = Power(Base=OneMinus(None=@n3)) {constExponent=5}
  n5 = Power(Base=@n3) {constExponent=5}
  n6 = Power(Base=OneMinus(None=Add(A=@n4, B=@n5))) {constExponent=10}
  n7 = Panner(Coordinate=@n2) {speedX=-2.5}
  n8 = Multiply(
   A=Power(Base=TextureSample(UVs=@n1) {texture=/DealershipSharedLibrary/FX/Textures/Noise/T_VCVFX_CloudDistort_A}.RGB, Exp=Constant() {r=0.5})
   B=Multiply(
    A=Multiply(A=@n6) {constB=2}
    B=TextureSample(UVs=@n7) {texture=/DealershipSharedLibrary/FX/Textures/TexturePackage/T_VCVFX_SmokeTrail01_RGB}.R
   )
  )
  n9 = ComponentMask(None=TextureCoordinate()) {r=true}
  n10 = Power(Base=@n9, Exp=Constant() {r=1})
  n11 = TextureCoordinate()
  n12 = Power(Base=OneMinus(None=ComponentMask(None=@n11) {r=true}), Exp=Constant() {r=2})
  n13 = Multiply(A=Multiply(A=@MaterialExpressionVertexColor_1.A, B=Multiply(A=@n8) {constB=3}), B=Multiply(A=Multiply(A=@n10, B=@n12), B=Constant() {r=15}))
  n14 = ComponentMask(None=TextureCoordinate()) {g=true}
  n15 = Constant() {r=4}
  n16 = OneMinus(None=ComponentMask(None=TextureCoordinate()) {g=true})
  n17 = Constant() {r=4}
  n18 = Multiply(A=Multiply(A=Power(Base=@n14, Exp=@n15), B=Power(Base=@n16, Exp=@n17)), B=Constant() {r=200})
  MaterialExpressionConstantClamp_5 = Saturate(None=MaterialFunctionCall(Opacity In=Multiply(A=@n13, B=@n18), DepthFade Distance=Constant() {r=50}) {materialFunction=&MF_VCVFX_DepthFade}.Result)
  n19 = DynamicParameter() {paramNames [StarsPaint,LBright,Param3,Param4]}
  n20 = Time()
  n21 = Constant() {r=0.2}
  n22 = TextureSample(UVs=Panner(Coordinate=TextureCoordinate(), Time=Multiply(A=@n20, B=@n21)) {speedX=-2}) {texture=/DealershipSharedLibrary/FX/Textures/Noise/T_VCVFX_Flow03_N}
  n23 = Add(
   A=LinearInterpolate(A=TextureCoordinate() {uTiling=1.5}, B=TextureCoordinate() {uTiling=0.5})
   B=Multiply(A=ComponentMask(None=@n22.RGB) {g=true, r=true}, B=Constant() {r=0.1})
  )
  n24 = Panner(Coordinate=@n23) {speedX=-1.5}
  n25 = Panner(Coordinate=@n23) {speedX=-1.25}
  n26 = Add(
   A=Add(
    A=TextureSample(UVs=@n24) {texture=/DealershipSharedLibrary/FX/Textures/TexturePackage/T_VCVFX_LightningstrailEDrift_RGB}.R
    B=TextureSample(UVs=@n25) {texture=&T_VCVFX_LightningstrailEDrift_RGB}.G
   )
   B=TextureSample(UVs=Panner(Coordinate=@n23) {speedX=-1.5}) {texture=&T_VCVFX_LightningstrailEDrift_RGB}.B
  )
  n27 = Multiply(A=@n19.LBright, B=LinearInterpolate(A=SmoothStep(Value=@n26) {constMin=0.7}, B=SmoothStep(Value=@n26) {constMin=0.5}))
  n28 = Multiply(A=Saturate(None=@MaterialExpressionVertexColor_1.RGB), B=@n27)
  LinearInterpolate(B=Multiply(A=Desaturation(None=@n28), B=@n27), Alpha=@n19.StarsPaint)
  n29 = Multiply(A=Saturate(None=@n28), B=ScalarParameter() {defaultValue=10500, parameterName=Bolt Intensity}) {constB=5500}
  n30 = Multiply(A=Power(Base=@n9, Exp=Constant() {r=1}), B=Power(Base=OneMinus(None=@n9)) {constExponent=8})
  MaterialFunctionCall(Emissive=Add(A=Multiply(A=@MaterialExpressionVertexColor_1.RGB, B=@n8), B=Multiply(A=@n29, B=@n30))) {materialFunction=&MF_FX_EmissiveAdjustments}
  Sine(None=Time()) {period=0.5}
 ]
}
