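# Translucent unlit Niagara ribbon material for the secondary trail of the ElectroGalactic drift trail vehicle cosmetic.
# A panning gradient mask combines with two scrolling cloud textures, one UV-distorted by billow noise, and falloffs along U.
# Time-driven sines blend two opacity shapes and two brightness levels to pulse the trail.
# Particle color tints emissive and its alpha scales opacity, which passes through SmoothStep and depth fade.
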Material /VehicleCosmetics/DriftTrail/DriftTrail_ElectroGalactic/Materials/M_Trail_Secondary_Trail
graph {
 properties {
  bAutomaticallySetUsageInEditor=false
  bUsedWithNiagaraRibbons=true
  blendMode=BLEND_Translucent
  shadingModel=MSM_Unlit
  usedShadingModels=Unlit
 }
 nodes [
  MaterialExpressionVertexColor_0 = ParticleColor()
  n1 = TextureSample(UVs=Panner(Coordinate=TextureCoordinate()) {speedX=-1.5}) {texture=/DealershipSharedLibrary/FX/Textures/Gradient_Masks/T_VCVFX_TMaskDrift_A}
  n2 = Panner(Coordinate=TextureCoordinate() {uTiling=0.1}) {speedX=-0.5, speedY=0.1}
  n3 = Multiply(
   A=ComponentMask(None=TextureSample(UVs=@n2) {texture=/DealershipSharedLibrary/FX/Textures/Noise/T_VCVFX_BillowNoiseTiledOffset01_RGB}.RGB) {g=true, r=true}
   B=Constant() {r=0.05}
  )
  n4 = TextureCoordinate()
  n5 = TextureSample(UVs=Panner(Coordinate=Add(A=@n3, B=@n4)) {speedX=-0.4, speedY=0.2}) {texture=/DealershipSharedLibrary/FX/Textures/Noise/T_VCVFX_CloudDistort_A}
  n6 = TextureCoordinate()
  n7 = Multiply(
   A=@n5.R
   B=TextureSample(UVs=Panner(Coordinate=@n6) {speedX=-0.7, speedY=-0.2}) {texture=&T_VCVFX_CloudDistort_A}.R
  )
  n8 = ComponentMask(None=TextureCoordinate()) {r=true}
  n9 = Power(Base=OneMinus(None=@n8)) {constExponent=4}
  n10 = Multiply(A=@n1.R, B=Multiply(A=@n7, B=Multiply(A=@n9) {constB=25}))
  n11 = Constant() {r=0.6}
  n12 = OneMinus(None=ComponentMask(None=TextureCoordinate()) {r=true})
  n13 = LinearInterpolate(A=Constant() {r=1}, B=Constant() {r=4})
  n14 = Multiply(A=Multiply(A=Power(Base=@n8, Exp=@n11), B=Power(Base=@n12, Exp=@n13)), B=LinearInterpolate(A=Constant() {r=6}, B=Constant() {r=2}))
  n15 = Multiply(A=@n1.R, B=Multiply(A=@n7, B=@n14))
  n16 = MaterialFunctionCall!(Sine Phase=Multiply(A=Time()) {constB=0.5}, Value 1=Constant(), Value 2=Constant() {r=0.8}) {materialFunction=/Engine/Functions/Engine_MaterialFunctions02/Utility/Sine_Remapped}
  n17 = Multiply(A=Multiply(A=@MaterialExpressionVertexColor_0.A, B=LinearInterpolate(A=@n10, B=@n15, Alpha=@n16.Remapped Sine) {constAlpha=1}), B=@n14)
  MaterialExpressionConstantClamp_0 = Saturate(None=MaterialFunctionCall(Opacity In=SmoothStep(Value=@n17)) {materialFunction=&MF_VCVFX_DepthFade}.Result)
  LinearInterpolate() {constA=0.75}
  n18 = Multiply(A=@MaterialExpressionVertexColor_0.RGB, B=Multiply(A=@n1.R, B=@n5.R))
  n19 = Multiply(A=@n18) {constB=7}
  n20 = Saturate(None=Multiply(A=@n18) {constB=6})
  n21 = MaterialFunctionCall(
   Sine Phase=Multiply(A=Time(), B=LinearInterpolate() {constA=0.3, constAlpha=0.6, constB=0.3}) {constB=0.3}
   Value 1=Constant() {r=0.8}
   Value 2=Constant() {r=1}
  ) {materialFunction=&Sine_Remapped}
  MaterialFunctionCall(Emissive=Multiply(A=LinearInterpolate(A=@n19, B=@n20, Alpha=@n21.Remapped Sine) {constAlpha=1}) {constB=1.5}) {materialFunction=&MF_FX_EmissiveAdjustments}
  n22 = TextureCoordinate()
  Multiply(A=@n18, B=OneMinus(None=ComponentMask(None=@n22) {r=true}))
 ]
}
