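# Additive unlit Niagara mesh and ribbon material for the wheel-wrap part of the ElectroGalactic drift trail, drawing an electric, smoky streak.
# It sums a panning, sine-animated lightning core, a noise-broken smoke band faded at the V edges, and flow-distorted, smoothstepped bolts masked by that smoke and UV falloffs.
# Particle color tints the layers, while dynamic parameters drive core speed and brightness plus bolt intensity.
# Opacity uses depth fade and a rounded V-band mask.
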Material /VehicleCosmetics/DriftTrail/DriftTrail_ElectroGalactic/Materials/M_Trail_Wheelwrap
graph {
 properties {
  bAutomaticallySetUsageInEditor=false
  bUsedWithNiagaraMeshParticles=true
  bUsedWithNiagaraRibbons=true
  blendMode=BLEND_Additive
  shadingModel=MSM_Unlit
  usedShadingModels=Unlit
 }
 nodes [
  n1 = Sine(None=Time()) {period=0.5}
  n2 = Multiply(
   A=TextureCoordinate() {vTiling=0.6}
   B=MaterialFunctionCall(X=Constant() {r=1}, Y=LinearInterpolate(Alpha=@n1) {constA=1, constB=0.9}) {materialFunction=&MakeFloat2}.Result
  )
  n3 = Panner(Coordinate=@n2) {speedX=-2.5}
  n4 = MaterialFunctionCall(Speed=Constant() {r=7}) {materialFunction=&TimeWithSpeedVariable}
  n5 = Constant4Vector() {constant=<rgba -1 1 0.25 0.1>}
  n6 = MaterialFunctionCall(Y=MaterialFunctionCall(Input=Sine(None=@n4.Time) {period=2}, Input Low=@n5.R, Input High=@n5.R, Target Low=@n5.R, Target High=@n5.R) {materialFunction=&RemapValueRange}.Result) {materialFunction=&MakeFloat2}
  n7 = Sine(None=Time() {bOverride_Period=true, period=1})
  n8 = Constant() {r=1}
  n9 = DynamicParameter() {parameterIndex=2, paramNames [CenterVelocity, LBright, CenterBrightness, BoltIntensity]}
  n10 = Panner(
   Coordinate=LinearInterpolate(A=TextureCoordinate() {uTiling=3, vTiling=0.6}, B=TextureCoordinate() {uTiling=1.5, vTiling=0.6}, Alpha=@n7) {constAlpha=1}
   Speed=Multiply(
    A=MaterialFunctionCall(X=@n8) {materialFunction=&MakeFloat2}.Result
    B=@n9.CenterVelocity
   )
  ) {speedX=-2.5, speedY=0.2}
  n11 = Add(
   A=TextureSample(UVs=Add(A=@n3, B=@n6.Result)) {texture=/DealershipSharedLibrary/FX/Textures/TexturePackage/T_VCVFX_LightningstrailEDrift_RGB}.R
   B=TextureSample(UVs=Add(A=@n10, B=@n6.Result)) {texture=&T_VCVFX_LightningstrailEDrift_RGB}.G
  )
  n12 = TextureSample(UVs=Add(A=@n6.Result, B=Panner(Coordinate=@n2) {speedX=-4})) {texture=&T_VCVFX_LightningstrailEDrift_RGB}
  MaterialExpressionVertexColor_1 = ParticleColor()
  n13 = Multiply(
   A=LinearInterpolate(A=@n11, B=Add(A=@n11, B=@n12.B), Alpha=@n7) {constAlpha=1}
   B=Multiply(A=@MaterialExpressionVertexColor_1.RGB, B=@n9.CenterBrightness) {constB=10}
  )
  n14 = Panner(Coordinate=TextureCoordinate(), Time=Constant() {r=2}) {speedX=2, speedY=0.1}
  n15 = ComponentMask(None=TextureCoordinate()) {g=true}
  n16 = Power(Base=OneMinus(None=@n15)) {constExponent=5}
  n17 = Power(Base=@n15) {constExponent=5}
  n18 = Power(Base=OneMinus(None=Add(A=@n16, B=@n17))) {constExponent=4}
  n19 = Sine(None=MaterialFunctionCall(Speed=Constant() {r=20}) {materialFunction=&TimeWithSpeedVariable}.Time) {period=10}
  n20 = Constant4Vector() {constant=<rgba -1 1 0.4 0>}
  n21 = Add(
   A=TextureCoordinate() {vTiling=0.5}
   B=MaterialFunctionCall(
    X=Constant()
    Y=MaterialFunctionCall(Input=@n19, Input Low=@n20.R, Input High=@n20.R, Target Low=@n20.R, Target High=@n20.R) {materialFunction=&RemapValueRange}.Result
   ) {materialFunction=&MakeFloat2}.Result
  )
  n22 = Multiply(
   A=Power(Base=TextureSample(UVs=@n14) {texture=/DealershipSharedLibrary/FX/Textures/Noise/T_VCVFX_CloudDistort_A}.RGB, Exp=Constant() {r=0.5})
   B=Multiply(
    A=Multiply(A=@n18) {constB=2}
    B=TextureSample(UVs=@n21) {texture=/DealershipSharedLibrary/FX/Textures/TexturePackage/T_VCVFX_SmokeTrail01_RGB}.R
   )
  )
  n23 = Multiply(A=@MaterialExpressionVertexColor_1.RGB, B=@n22)
  n24 = Time()
  n25 = Constant() {r=0.2}
  n26 = TextureSample(UVs=Panner(Coordinate=TextureCoordinate(), Time=Multiply(A=@n24, B=@n25)) {speedX=-2}) {texture=/DealershipSharedLibrary/FX/Textures/Noise/T_VCVFX_Flow03_N}
  n27 = Add(A=TextureCoordinate() {uTiling=5}, B=Multiply(A=ComponentMask(None=@n26.RGB) {g=true, r=true}, B=Constant() {r=0.1}))
  n28 = Panner(Coordinate=@n27) {speedX=2, speedY=2}
  n29 = Panner(Coordinate=@n27) {speedX=2.25}
  n30 = Add(
   A=Add(
    A=TextureSample(UVs=@n28) {texture=&T_VCVFX_LightningstrailEDrift_RGB}.R
    B=TextureSample(UVs=@n29) {texture=&T_VCVFX_LightningstrailEDrift_RGB}.G
   )
   B=TextureSample(UVs=Panner(Coordinate=@n27) {speedX=2.5, speedY=-2}) {texture=&T_VCVFX_LightningstrailEDrift_RGB}.B
  )
  n31 = SmoothStep(Value=@n30) {constMin=0.7}
  n32 = SmoothStep(Value=@n30) {constMin=0.5}
  n33 = Multiply(A=Saturate(None=@MaterialExpressionVertexColor_1.RGB), B=Multiply(A=@n9.LBright, B=LinearInterpolate(A=@n31, B=@n32)))
  n34 = Power(Base=ComponentMask(None=TextureCoordinate()) {r=true}, Exp=Constant() {r=2})
  n35 = TextureCoordinate()
  n36 = Power(Base=OneMinus(None=ComponentMask(None=@n35) {r=true}), Exp=Constant() {r=2})
  n37 = Multiply(A=@n22, B=Multiply(A=Multiply(A=@n33, B=@n9.BoltIntensity) {constB=5500}, B=Multiply(A=@n34, B=@n36)))
  n38 = ComponentMask(None=TextureCoordinate()) {g=true}
  n39 = Multiply(A=@n38, B=OneMinus(None=@n38))
  n40 = Floor(None=Multiply(A=Power(Base=@n39)) {constB=10000})
  n41 = Multiply(A=@n37, B=@n40)
  MaterialFunctionCall(Emissive=Add(A=@n13, B=Add(A=@n23, B=@n41))) {materialFunction=&MF_FX_EmissiveAdjustments}
  n42 = Add(A=@n13, B=Add(A=@n22, B=Multiply(A=@n40, B=@n37)))
  MaterialExpressionConstantClamp_5 = Saturate(None=MaterialFunctionCall(Opacity In=Multiply(A=@MaterialExpressionVertexColor_1.A, B=@n42), DepthFade Distance=Constant() {r=20}) {materialFunction=&MF_VCVFX_DepthFade}.Result)
  n43 = ComponentMask(None=TextureCoordinate()) {g=true}
  n44 = Multiply(A=Multiply(A=OneMinus(None=@n43), B=@n43)) {constB=4.5}
  Multiply(A=@MaterialExpressionConstantClamp_5, B=Round(None=Power(Base=@n44) {constExponent=1}))
 ]
}
